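; Grapple Glider weapon Blueprint that drives reticle feedback showing whether a grapple target is in range.
; While equipped, a looping timer runs CheckForTrees, which line-traces from the player's view along the targeting range minus an offset.
; A hit shows the in-range center reticle and hides the red X; a miss hides the center and shows the X if the point under the reticle is out of range.
; Unequipping clears the timer.
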(event EventBeginPlay
  (|CallParentFunction))

(event EventTick
  (params (DeltaSeconds :type Float%20%28single-precision%29))
  (|CallParentFunction DeltaSeconds) ; This node is disabled and will not be called.\nDrag off pins to build functionality.
)

(event Custom|CheckForTrees
  (bind _playerpawn (Variables|Default|GetPlayerPawn)
    (:then
      (bind (_outhit _returnvalue) (Collision|LineTraceForObjects (Actor|GetActorEyesViewPoint _playerpawn) (+ (Actor|GetActorEyesViewPoint _playerpawn) (* (- (ScalableFloat|GetValueatLevel (Variables|Default|GetTargetingRange)) (Variables|Default|GetTargetingOffset)) (Math|Vector|GetForwardVector (.OutRotation (Actor|GetActorEyesViewPoint _playerpawn))))) (Utilities|Array|MakeArray "ObjectTypeQuery1")))
      (if (not _returnvalue)
        (Reticle|EnableAltCenterReticle false) ; Turn off the "in range" visual
        (if (not (<= (Math|Vector|Distance(Vector) (Actor|GetActorEyesViewPoint _playerpawn) (Reticle|GetLocationUnderReticle (Fortnite|GetFortPlayerControllerfromActor _playerpawn))) (ScalableFloat|GetValueatLevel (Variables|Default|GetTargetingRange))))
          (Reticle|EnableAltOuterReticle true) ; Enable the red X
        ))
      (Reticle|EnableAltCenterReticle true) ; Turn on the "in range" visual
      (Reticle|EnableAltOuterReticle false) ; Disable the red X
    )
    (:else)))

(event FortWeapon|EventOnEquippedAndAttached
  (|CallParentFunction)
  (bind _returnvalue (Utilities|Time|SetTimerbyEvent (EventDispatchers|CreateEvent :!event Custom|CheckForTrees) 0.1 true))
  (Variables|Default|SetCheckForTrees_Timer _returnvalue)
  (bind _asfort_player_pawn (Utilities|Casting|CastToFortPlayerPawn (Game|GetInstigator)))
  (Variables|Default|SetPlayerPawn _asfort_player_pawn))

(event FortWeapon|EventOnUnEquippedAndDetached
  (|CallParentFunction)
  (Utilities|Time|ClearandInvalidateTimerbyHandle (Variables|Default|GetCheckForTrees_Timer)))

(fn ConstructionScript
  (|CallParentFunction))
