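; Grapple Glider projectile that detects world-static hits, such as trees, and starts the player's grapple travel.
; A self-resetting timer sphere-traces from the cached previous location, allowing checks less often than every frame.
; Trace hits stop movement and tether the pawn, with the server triggering Travel.
; Collision hits cancel the timer and dismount riders or eject open-seat passengers, while SpawnRope creates the rope actor linked to weapon and projectile.
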(event EventBeginPlay
  (|CallParentFunction)
  (Variables|Default|SetPreviousFrameLocation (Transformation|GetActorLocation))
  (CallFunction|SetCheckforTreesTimer))

(event Custom|SetCheckForTreesTimer
  (bind _checkfortreesfrequency (Variables|Default|GetCheckForTreesFrequency))
  (bind _returnvalue (ScalableFloat|GetValueatLevel _checkfortreesfrequency))
  (bind _outputdelegate (EventDispatchers|CreateEvent :!event Custom|CheckForTrees))
  ;[0.7026 0.1455 1] Set the timer for the trees check - if the check frequency is zero or lower, just update once every frame. Don't loop the timer, it'll be reset manually by CheckForTrees.
  (if (<= _returnvalue 0)
    (bind _returnvalue_2 (Utilities|Time|SetTimerforNextTickbyEvent _outputdelegate))
    (Variables|Default|SetCheckForTrees_Timer _returnvalue_2)
    (else
      (bind _returnvalue_3 (Utilities|Time|SetTimerbyEvent _outputdelegate _returnvalue))
      (Variables|Default|SetCheckForTrees_Timer _returnvalue_3))))

(event Custom|CheckForTrees
  (Utilities|FlowControl|Sequence
    (:then_0
      ;[0.08239 0.3963 1] Trace for world static objects at the location of the projectile and stop the projectile if it hits something.
      (Variables|Default|SetbCheckForTrees true)
      (bind _outhits (Collision|MultiSphereTraceForObjects (Variables|Default|GetPreviousFrameLocation) (Transformation|GetActorLocation) 15.0 (Utilities|Array|MakeArray "ObjectTypeQuery1")))
      (bind _array_element (Utilities|Array|ForEachLoopwithBreak _outhits)
        (:LoopBody
          (bind _playerpawn (Variables|Default|GetPlayerPawn)
            (:then
              (Tether|SetTetherPawn _playerpawn 0))
            (:else
              _))
          (Components|Movement|StopMovementImmediately (Variables|Movement|GetProjectileMovementComponent))
          (CallFunction|HandleHits _array_element)
          (Actor|SetLifeSpan (Variables|Default|GetLifespan%20on%20Hit))
          (if (and (Networking|IsServer) (Utilities|IsValid (Game|GetInstigatorController) :!class /Script/Engine.KismetSystemLibrary))
            (CallFunction|Travel (.Location (Collision|BreakHitResult _array_element)))
            (Variables|Default|SetbCheckForTrees false)))))
    (:then_1
      ;[0.03839 1 0.3633] Set the timer for the next check (it doesn't loop automatically).
      (if (Variables|Default|GetbCheckForTrees)
        (CallFunction|SetCheckforTreesTimer)
        ;[1 0.2904 0.7663] Cache the current location to use as the trace start location for the next trees check. This allows the trees check to occur less frequently than once per frame.
        (Variables|Default|SetPreviousFrameLocation (Transformation|GetActorLocation))))))

(event Collision|EventHit
  (params (MyComp :type Primitive%20Component%20Object%20Reference) (Other :type Actor%20Object%20Reference) (OtherComp :type Primitive%20Component%20Object%20Reference) (bSelfMoved :type Boolean) (HitLocation :type Vector :default "0, 0, 0") (HitNormal :type Vector :default "0, 0, 0") (NormalImpulse :type Vector :default "0, 0, 0") (Hit :type Hit%20Result%20Structure%20%28by%20ref%29))
  (bind _playerpawn (Variables|Default|GetPlayerPawn))
  (bind _returnvalue (Actor|GetComponentbyClass _playerpawn "/Script/RidingCodeRuntime.RiderComponent"))
  (bind _returnvalue_2 (Vehicle|GetVehicle _playerpawn))
  (Utilities|Time|ClearandInvalidateTimerbyHandle (Variables|Default|GetCheckForTrees_Timer))
  (if (GameplayTags|HasMatchingGameplayTag _playerpawn "(TagName=\"Granted.Riding.Player.IsRiding\")")
    (Utilities|IsValid _returnvalue
      (:Is%20Valid
        (Riding|StopRidingActor _returnvalue false false true true))
      (:Is%20Not%20Valid
        _))
    (elif (GameplayTags|HasMatchingGameplayTag _playerpawn "(TagName=\"Gameplay.Status.InVehicle.Seat.Open\")")
      (Utilities|IsValid _returnvalue_2
        (:Is%20Valid
          (Vehicle|EjectPlayerPawn(Message) _returnvalue_2 _playerpawn)))
      (else
        _)))
  (|CallParentFunction MyComp Other OtherComp bSelfMoved HitLocation HitNormal NormalImpulse Hit))

(fn SpawnRope
  (params (PlayerPawn :type Fort%20Player%20Pawn%20Object%20Reference))
  ;[1 1 1] Spawn The Rope
  (bind _returnvalue (Game|SpawnActorfromClass "/GrappleGlider/Items/B_HookGunRope_GrappleGlider.B_HookGunRope_GrappleGlider_C" (Math|Transform|MakeTransform (Transformation|GetActorLocation)) "Undefined" "MultiplyWithRoot" (Variables|Default|GetWeapon%20Actor) PlayerPawn 0))
  (bind _output_get (Variables|Default|SetRope _returnvalue))
  (Class|FortHookGunRope|SetWeaponandProjectile _output_get (Variables|Default|GetWeapon%20Actor) self)
  (return _output_get))

(fn ConstructionScript
  (|CallParentFunction))
